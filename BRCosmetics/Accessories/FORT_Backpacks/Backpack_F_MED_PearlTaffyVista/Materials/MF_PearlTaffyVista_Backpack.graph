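# Material function for the PearlTaffyVista backpack that adds an animated pink emissive effect and glitter, outputting the material and an effect mask.
# A mask derived from base color darkness combines with a sine pulse and moving rim bands.
# Bands come from sphere masks over time, using radial pre-skinned coordinates and panning noise.
# Frontend-loop audio amplitude adds reactivity; several debug and texture nodes are left unconnected.
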MaterialFunction /BRCosmetics/Accessories/FORT_Backpacks/Backpack_F_MED_PearlTaffyVista/Materials/MF_PearlTaffyVista_Backpack
graph {nodes [
  n1 = GetMaterialAttributes(None=NamedRerouteUsage()) {attributeGetTypes [#1]}
  n2 = NamedRerouteUsage()
  n3 = Max(A=NamedRerouteUsage(), B=Multiply(A=NamedRerouteUsage()) {constB=0.1})
  n4 = NamedRerouteUsage()
  FunctionOutput(None=MaterialFunctionCall!(
    In=SetMaterialAttributes!(MaterialAttributes=@n1.MaterialAttributes, EmissiveColor=@n2) {attributeSetTypes [#2]}
    SpeedMultiplier
    GlitterPower
    GlitterIntensity
    GlitterColor
    SkinFX_Mask=Multiply(A=@n3, B=@n4)
   ) {materialFunction=&/MFI_PearlTaffyVista_Glitter}.Result) {outputName=Material}
  n5 = ScalarParameter() {defaultValue=0.5, parameterName=A}
  n6 = ScalarParameter() {defaultValue=2, parameterName=B_1}
  n7 = Power(Base=NamedRerouteUsage()) {constExponent=1.5}
  n8 = Multiply(
   A=Add(A=NamedRerouteUsage(), B=ScalarParameter() {defaultValue=0.65, parameterName=B_5})
   B=ScalarParameter() {defaultValue=1.25, parameterName=B_3}
  ) {constB=2}
  n9 = Add(A=@n8, B=Multiply(A=Time()) {constB=-0.5})
  n10 = Sine(None=@n9)
  n11 = LinearInterpolate(A=ScalarParameter() {defaultValue=0.1, parameterName=A_1}, Alpha=Power(Base=ConstantBiasScale(None=@n10)) {constExponent=1}) {constA=0.1}
  n12 = Multiply(
   A=Multiply(A=NamedRerouteUsage(), B=LinearInterpolate(A=@n5, B=@n6, Alpha=@n7) {constA=0.1})
   B=Max(A=NamedRerouteDeclaration(None=@n11) {name=EffectMask, nodeColor=<rgba 0 1 0.7422 1>}, B=NamedRerouteUsage())
  )
  n13 = Multiply(A=@n12, B=NamedRerouteUsage())
  FunctionOutput(None=@n13) {outputName=EffectMask, sortPriority=1}
  n14 = VectorParameter() {defaultValue=<rgba 0.4844 0.094 0.2211 1>}
  LinearInterpolate(A=NamedRerouteUsage(), B=@n14.RGB, Alpha=@n12)
  NamedRerouteDeclaration(None=FunctionInput() {inputType=FunctionInput_MaterialAttributes}) {name=Material In, nodeColor=<rgba 0.3125 1 0 1>}
  NamedRerouteDeclaration(None=@n1.BaseColor) {name=BaseColor In, nodeColor=<rgba 1 0 0.6328 1>}
  n15 = MaterialFunctionCall!(InAmplitude, InSpectrum, CustomFakeSpectrum, CustomFakeAmplitude) {materialFunction=/BRCosmetics/Global/MaterialFunctions/AudioReactivity/MF_AudioReactivity_FrontendLoop}
  n16 = ScalarParameter() {defaultValue=2, parameterName=Amplitude Exponent}
  NamedRerouteDeclaration(None=Multiply(A=Power(Base=@n15.Amplitude, Exp=@n16), B=ScalarParameter() {defaultValue=1, parameterName=AudioReactivity On})) {name=Amplitude, nodeColor=<rgba 1 0.7266 0 1>}
  NamedRerouteDeclaration(None=Multiply(A=@n14.RGB, B=@n13)) {name=Emissive Out, nodeColor=<rgba 0 0.9531 1 1>}
  n17 = Desaturation(None=NamedRerouteUsage(), Fraction=Constant() {r=1})
  n18 = Multiply(A=@n17, B=@n17)
  NamedRerouteDeclaration(None=OneMinus(None=SmoothStep(Value=@n18) {constMax=0.5})) {name=ColorMask, nodeColor=<rgba 0 1 0.7188 1>}
  n19 = Add(A=LocalPosition() {localOrigin=InstancePreSkinning}.XYZ, B=Constant3Vector() {constant=<rgba 0 0 5 0>})
  n20 = Multiply(A=NamedRerouteDeclaration(None=VertexInterpolator(VS=@n19).PS) {name=PreSkinnedPosition, nodeColor=<rgba 0 0.4375 1 1>}) {constB=0.01}
  n21 = ScalarParameter() {defaultValue=1.5, parameterName=B}
  n22 = MaterialFunctionCall(Vector or UVs=ComponentMask(None=Multiply(A=@n20, B=@n21)) {b=true, r=true}) {materialFunction=&VectorToRadialValue}
  n23 = Multiply(A=TextureCoordinate()) {constB=1.5}
  n24 = Subtract(A=TextureSample(UVs=Panner(Coordinate=@n23) {speedX=0.01, speedY=-0.1}) {samplerType=SAMPLERTYPE_Grayscale, texture=&LowResBlurredNoise}.R) {constB=0.5}
  n25 = TextureCoordinate()
  n26 = TextureSample!(
   UVs=Add(A=Panner(Coordinate=@n22.Radial Coordinates) {speedX=0.01, speedY=-0.2}, B=Multiply(A=@n24) {constB=0.05})
   Tex
   "DDX(UVs)"=DDX(Value=@n25)
   "DDY(UVs)"=DDY(Value=@n25)
   Apply View MipBias
  ) {mipValueMode=TMVM_Derivative, samplerType=SAMPLERTYPE_Grayscale, texture=&LowResBlurredNoise}
  NamedRerouteDeclaration(None=@n26.R) {name=noise, nodeColor=<rgba 1 0.3984 0 1>}
  NamedRerouteDeclaration(None=@n22.Linear Distance) {name=LinearDistance, nodeColor=<rgba 0.7578 1 0 1>}
  n27 = NamedRerouteUsage()
  n28 = Add(A=@n8, B=Multiply(A=Subtract(A=@n27) {constB=0.5}, B=ScalarParameter() {defaultValue=0.35, parameterName=B_4}) {constB=0.1})
  n29 = ScalarParameter() {defaultValue=0.4}
  n30 = ScalarParameter() {defaultValue=0.2, parameterName=Radius}
  n31 = SphereMask(A=Frac(None=Add(A=@n9, B=@n28)), B=@n29, Radius=@n30) {attenuationRadius=0.02, hardnessPercent=0}
  n32 = ScalarParameter() {defaultValue=0.2, parameterName=A_2}
  n33 = ScalarParameter() {defaultValue=0.8, parameterName=B_6}
  n34 = Power(Base=NamedRerouteUsage()) {constExponent=3}
  n35 = Frac(None=Add(A=Multiply(A=@n28), B=LinearInterpolate(A=@n32, B=@n33, Alpha=@n34) {constA=0.05, constB=1.05}))
  n36 = Multiply(
   A=Multiply(A=SphereMask(A=@n35, B=@n29, Radius=@n30) {attenuationRadius=0.02, hardnessPercent=0}, B=LinearInterpolate(Alpha=@n34) {constA=0.75})
   B=ScalarParameter() {defaultValue=0.75, parameterName=B_7}
  )
  NamedRerouteDeclaration(None=Multiply(A=Add(A=@n31, B=@n36), B=ScalarParameter() {defaultValue=2, parameterName=B_2})) {name=RimEdge, nodeColor=<rgba 0 1 0.7656 1>}
  NamedRerouteUsage()
  n37 = MaterialFunctionCall(Speed=Constant() {r=1.5}) {materialFunction=&DebugTimeSine}
  n38 = MaterialFunctionCall() {materialFunction=&DebugTimeSine}
  n39 = Constant() {r=1.25}
  Saturate(None=Subtract(
    A=Add(A=@n37.Result, B=@n38.Result)
    B=MaterialFunctionCall(Speed=@n39) {materialFunction=&DebugTimeSine}.Result
   ))
  ScalarParameter() {defaultValue=0.2, parameterName=Param_1}
  TextureSample!(UVs, Tex, Apply View MipBias) {texture=/BRCosmetics/Accessories/FORT_Backpacks/Backpack_F_MED_PearlTaffyVista/Textures/T_PearlTaffyVista_Backpack_D}
  TextureSample!(UVs, Tex, "DDX(UVs)", "DDY(UVs)", Apply View MipBias) {mipValueMode=TMVM_Derivative, texture=/BRCosmetics/Characters/Player/Male/Medium/Bodies/M_MED_Patches/Textures/T_Smoke_01}
 ]}
