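# Translucent parent material for the soft glow outline on the PearlTaffyVista backpack cosmetic, built for skeletal meshes and allowed to write custom depth.
# MF_SoftGlowOutlines supplies the material attributes, with only Opacity overridden by multiplying it with an EffectMask.
# That mask comes from MF_PearlTaffyVista_Backpack, fed by the 2023 character parent PreFX function.
# The masked attributes and PreFX data then pass through the matching PostFX function.
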Material /BRCosmetics/Accessories/FORT_Backpacks/Backpack_F_MED_PearlTaffyVista/Materials/M_PearlTaffyVista_Backpack_SoftGlow_Parent
graph {
 properties {
  allowTranslucentCustomDepthWrites=true
  bAutomaticallySetUsageInEditor=false
  bUseMaterialAttributes=true
  bUsedWithMeshDeformer=true
  bUsedWithMorphTargets=true
  bUsedWithSkeletalMesh=true
  blendMode=BLEND_Translucent
  pixelDepthOffsetMode=PDOM_Legacy
  previewMesh=/BRCosmetics/Accessories/FORT_Backpacks/Backpack_F_MED_PearlTaffyVista/Meshes/F_MED_PearlTaffyVista_Pack
  shadingModel=MSM_FromMaterialExpression
  usedShadingModels=Default Lit | Subsurface | Clear Coat | Subsurface Profile | Cloth
 }
 nodes [
  n1 = GetMaterialAttributes(None=MaterialFunctionCall() {materialFunction=/Game/Global/Materials/MaterialGym/Materials/SoftGlowOutlines/MF_SoftGlowOutlines}.Result) {attributeGetTypes [#1]}
  n2 = MaterialFunctionCall() {materialFunction=&MF_CharacterParent_2023_PreFX}
  n3 = MaterialFunctionCall!(In=@n2.Result) {materialFunction=&/MF_PearlTaffyVista_Backpack}
  MaterialFunctionCall(
   Material=SetMaterialAttributes!(MaterialAttributes=@n1.MaterialAttributes, Opacity=Multiply(A=@n1.Opacity, B=@n3.EffectMask)) {attributeSetTypes [#1]}
   M=@n2.M
   Pre Skinned Local Position=@n2.M
   Distance Blend=@n2.M
   Fresnel=@n2.M
  ) {materialFunction=&MF_CharacterParent_2023_PostFX}
 ]
}
