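# Physics asset for the PearlTaffyVista backpack cosmetic, defining bodies and joint limits for its two pairs of wings.
# Each upper and lower dyn wing bone has a large angled capsule, mirrored left to right and constrained to that side's MSTR_wing bone.
# The pack and MSTR_wing bones carry only tiny capsules, with no constraint linking them to each other.
# Upper wings get wider swing and twist limits than lower wings.
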PhysicsAsset /BRCosmetics/Accessories/FORT_Backpacks/Backpack_F_MED_PearlTaffyVista/Meshes/F_MED_PearlTaffyVista_Pack_Physics
bodies {
 MSTR_wing_l [{shapeType=Capsule, length=0.505, radius=0.505, shapeName=MSTR_wing_l_capsule, rotation=<rot 0 -90 0>}]
 MSTR_wing_r [{shapeType=Capsule, length=0.505, radius=0.505, shapeName=MSTR_wing_r_capsule, rotation=<rot 0 -90 0>}]
 dyn_down_wing_l [{
   shapeType=Capsule
   length=11.5292
   radius=7.2737
   shapeName=dyn_down_wing_l_capsule
   center=<8.5475 -5.8316 4.8271>
   rotation=<rot 0 -141.4869 -126.2338>
  }]
 dyn_down_wing_r [{
   shapeType=Capsule
   length=11.5292
   radius=7.2737
   shapeName=dyn_down_wing_r_capsule
   center=<8.5475 -5.8316 -4.8271>
   rotation=<rot 0 -141.4869 -53.7662>
  }]
 dyn_up_wing_l [{
   shapeType=Capsule
   length=18.7034
   radius=9.0914
   shapeName=dyn_up_wing_l_capsule
   center=<-7.5319 -10.2101 11.3647>
   rotation=<rot 0 123.2969 -122.0005>
  }]
 dyn_up_wing_r [{
   shapeType=Capsule
   length=18.7034
   radius=9.0914
   shapeName=dyn_up_wing_r_capsule
   center=<-7.5319 -10.2101 -11.3647>
   rotation=<rot 0 123.2969 -57.9995>
  }]
 pack [{shapeType=Capsule, length=0.505, radius=0.505, rotation=<rot 0 -90 0>}]
}
constraints [
 {bone1Name=dyn_down_wing_l, bone2Name=MSTR_wing_l, swing1LimitDegrees=45, swing2LimitDegrees=15, twistLimitDegrees=15}
 {bone1Name=dyn_up_wing_l, bone2Name=MSTR_wing_l, swing1LimitDegrees=60, swing2LimitDegrees=45, twistLimitDegrees=35}
 {bone1Name=dyn_down_wing_r, bone2Name=MSTR_wing_r, swing1LimitDegrees=45, swing2LimitDegrees=15, twistLimitDegrees=15}
 {bone1Name=dyn_up_wing_r, bone2Name=MSTR_wing_r, swing1LimitDegrees=60, swing2LimitDegrees=45, twistLimitDegrees=35}
]
