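# Animation Blueprint for the PearlTaffyVista back bling, animating its four wing bones over rigid-body dynamics.
# Update logic eases wing yaw by sprint, dive, jump/fall and crouch state, derives arm-based rotations and lateral-movement values, and selects front-end or game sim settings.
# The AnimGraph chains cached poses: RigidBody simulation in pack-bone space, state-driven yaw, then additive bone modifiers for upper-arm-following rotation, fixed wing offsets and pack adjustment.
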AnimBlueprint /BRCosmetics/Accessories/FORT_Backpacks/Backpack_F_MED_PearlTaffyVista/Meshes/F_MED_PearlTaffyVista_Pack_AnimBP
target_skeleton=&/F_MED_PearlTaffyVista_Pack_Skeleton
graphs [
 AnimGraph {nodes [
   n1 = "Transform (Modify) Bone - Bone: dyn_up_wing_l"(ComponentPose=Local To Component(LocalPose=Use cached pose 'Adjustments'()), Alpha=1, Translation="0,0,0", Rotation="4,12,0", Scale="1,1,1") {
    rotationMode=BMM_Additive
    alphaScaleBiasClamp {bInterpResult=true, interpSpeedIncreasing=8}
    boneToModify {boneName=dyn_up_wing_l}
    rotation {pitch=4, yaw=12}
   }
   n2 = "Transform (Modify) Bone - Bone: dyn_down_wing_l"(ComponentPose=@n1, Alpha=1, Translation="0,0,0", Rotation="2.961,14,0.2", Scale="1,1,1") {
    rotationMode=BMM_Additive
    alphaScaleBiasClamp {bInterpResult=true, interpSpeedIncreasing=8}
    boneToModify {boneName=dyn_down_wing_l}
    rotation=<rot 2.961 14 0.2>
   }
   n3 = "Transform (Modify) Bone - Bone: dyn_up_wing_r"(ComponentPose=@n2, Alpha=1, Translation="0,0,0", Rotation="-9.277,-10,5.191", Scale="1,1,1") {
    rotationMode=BMM_Additive
    alphaScaleBiasClamp {bInterpResult=true, interpSpeedIncreasing=8}
    boneToModify {boneName=dyn_up_wing_r}
    rotation=<rot -9.277 -10 5.191>
   }
   n4 = "Transform (Modify) Bone - Bone: dyn_down_wing_r"(ComponentPose=@n3, Alpha=1, Translation="0,0,0", Rotation="-4.596,-6,-3.127", Scale="1,1,1") {
    rotationMode=BMM_Additive
    alphaScaleBiasClamp {bInterpResult=true, interpSpeedIncreasing=8}
    boneToModify {boneName=dyn_down_wing_r}
    rotation=<rot -4.596 -6 -3.127>
   }
   n5 = "Transform (Modify) Bone - Bone: dyn_up_wing_l"(ComponentPose=@n4, Alpha=1, Translation="0,0,0", Rotation="10,-8,-2", Scale="1,1,1") {
    rotationMode=BMM_Additive
    alphaScaleBiasClamp {bInterpResult=true, interpSpeedIncreasing=8}
    boneToModify {boneName=dyn_up_wing_l}
    rotation=<rot 10 -8 -2>
   }
   n6 = "Transform (Modify) Bone - Bone: dyn_down_wing_l"(ComponentPose=@n5, Alpha=1, Translation="0,0,0", Rotation="7,-12,0.503", Scale="1,1,1") {
    rotationMode=BMM_Additive
    alphaScaleBiasClamp {bInterpResult=true, interpSpeedIncreasing=8}
    boneToModify {boneName=dyn_down_wing_l}
    rotation=<rot 7 -12 0.503>
   }
   n7 = "Transform (Modify) Bone - Bone: dyn_up_wing_r"(ComponentPose=@n6, Alpha=1, Translation="0,0,0", Rotation="-4,-9,3", Scale="1,1,1") {
    rotationMode=BMM_Additive
    alphaScaleBiasClamp {bInterpResult=true, interpSpeedIncreasing=8}
    boneToModify {boneName=dyn_up_wing_r}
    rotation=<rot -4 -9 3>
   }
   n8 = "Transform (Modify) Bone - Bone: dyn_down_wing_r"(ComponentPose=@n7, Alpha=1, Translation="0,0,0", Rotation="-2,-8,-1", Scale="1,1,1") {
    rotationMode=BMM_Additive
    alphaScaleBiasClamp {bInterpResult=true, interpSpeedIncreasing=8}
    boneToModify {boneName=dyn_down_wing_r}
    rotation=<rot -2 -8 -1>
   }
   Output Pose(Result=Component To Local(ComponentPose=@n8)) {layerGroup=None}
   n9 = RigidBody!(
    ComponentPose=Local To Component(LocalPose=Input Pose())
    Alpha
    ExternalForce
    SimSpaceSettings:Sim Space Settings Structure=Get Sim Space Settings() -> (Sim Space Settings:Sim Space Settings Structure)
   ) -> (Pose) {
    bClampLinearTranslationLimitToRefPose=true
    bOverrideWorldGravity=true
    overridePhysicsAsset=&/F_MED_PearlTaffyVista_Pack_Physics
    simulationSpace=BaseBoneSpace
    baseBoneRef {boneName=pack}
    overrideWorldGravity {z=-980}
    simSpaceSettings {
     dampingAlpha=0
     maxAngularAcceleration=40
     maxAngularVelocity=10
     maxLinearAcceleration=3200
     maxLinearVelocity=800
     velocityScaleZ=0.6
     worldAlpha=1
     externalLinearDragV {z=10}
    }
   }
   Save cached pose 'Dynamics'(Pose=Component To Local(ComponentPose=@n9)) {cacheName=Dynamics}
   n10 = Local To Component(LocalPose=Use cached pose 'Dynamics'())
   n11 = Get wing_up_r() -> (wing_up_r:double)
   n12 = "Transform (Modify) Bone - Bone: dyn_up_wing_l"!(
    ComponentPose="Transform (Modify) Bone - Bone: dyn_down_wing_r"!(
     ComponentPose="Transform (Modify) Bone - Bone: dyn_up_wing_r"!(ComponentPose=@n10, Alpha=1, Translation="0,0,0", Rotation_Roll:float=0, Rotation_Pitch:float=0, Rotation_Yaw:float=@n11, Scale="1,1,1") -> (Pose) {rotationMode=BMM_Additive, boneToModify {boneName=dyn_up_wing_r}}
     Alpha=1
     Translation="0,0,0"
     Rotation_Roll:float=0
     Rotation_Pitch:float=0
     Rotation_Yaw:float=Get wing_down_r() -> (wing_down_r:double)
     Scale="1,1,1"
    ) -> (Pose) {rotationMode=BMM_Additive, boneToModify {boneName=dyn_down_wing_r}}
    Alpha=1
    Translation="0,0,0"
    Rotation_Roll:float=0
    Rotation_Pitch:float=0
    Rotation_Yaw:float=Get wing_up_l() -> (wing_up_l:double)
    Scale="1,1,1"
   ) -> (Pose) {rotationMode=BMM_Additive, boneToModify {boneName=dyn_up_wing_l}}
   n13 = Get wing_down_l() -> (wing_down_l:double)
   Save cached pose 'Wings Animation'(Pose=Component To Local(ComponentPose="Transform (Modify) Bone - Bone: dyn_down_wing_l"!(ComponentPose=@n12, Alpha=1, Translation="0,0,0", Rotation_Roll:float=0, Rotation_Pitch:float=0, Rotation_Yaw:float=@n13, Scale="1,1,1") -> (Pose) {rotationMode=BMM_Additive, boneToModify {boneName=dyn_down_wing_l}})) {cacheName=Wings Animation}
   n14 = Use cached pose 'Wings Animation'()
   n15 = "Transform (Modify) Bone - Bone: dyn_down_wing_l"!(
    ComponentPose="Transform (Modify) Bone - Bone: dyn_up_wing_l"!(
     ComponentPose=Local To Component(LocalPose=@n14)
     Alpha=1
     Translation="0,0,0"
     Rotation:Rotator=Get Rotation Wing Up L() -> (Rotation Wing Up L:Rotator)
     Scale="1,1,1"
    ) -> (Pose) {rotationMode=BMM_Additive, alphaScaleBiasClamp {bInterpResult=true}, boneToModify {boneName=dyn_up_wing_l}}
    Alpha=1
    Translation="0,0,0"
    Rotation:Rotator=Get Rotation Wing Down L() -> (Rotation Wing Down L:Rotator)
    Scale="1,1,1"
   ) -> (Pose) {rotationMode=BMM_Additive, alphaScaleBiasClamp {bInterpResult=true}, boneToModify {boneName=dyn_down_wing_l}}
   n16 = Get Rotation Wing Up R() -> (Rotation Wing Up R:Rotator)
   n17 = "Transform (Modify) Bone - Bone: dyn_up_wing_l"(
    ComponentPose="Transform (Modify) Bone - Bone: dyn_down_wing_r"!(
     ComponentPose="Transform (Modify) Bone - Bone: dyn_up_wing_r"!(ComponentPose=@n15, Alpha=1, Translation="0,0,0", Rotation:Rotator=@n16, Scale="1,1,1") -> (Pose) {rotationMode=BMM_Additive, alphaScaleBiasClamp {bInterpResult=true}, boneToModify {boneName=dyn_up_wing_r}}
     Alpha=1
     Translation="0,0,0"
     Rotation:Rotator=Get Rotation Wing Down R() -> (Rotation Wing Down R:Rotator)
     Scale="1,1,1"
    ) -> (Pose) {rotationMode=BMM_Additive, alphaScaleBiasClamp {bInterpResult=true}, boneToModify {boneName=dyn_down_wing_r}}
    Alpha=1
    Translation="0,0,0"
    Rotation="-5,-30,0"
    Scale="1,1,1"
   ) {rotationMode=BMM_Additive, alphaScaleBiasClamp {bInterpResult=true}, boneToModify {boneName=dyn_up_wing_l}, rotation {pitch=-5, yaw=-30}}
   n18 = "Transform (Modify) Bone - Bone: dyn_down_wing_l"(ComponentPose=@n17, Alpha=1, Translation="0,0,0", Rotation="-5,-30,0", Scale="1,1,1") {rotationMode=BMM_Additive, alphaScaleBiasClamp {bInterpResult=true}, boneToModify {boneName=dyn_down_wing_l}, rotation {pitch=-5, yaw=-30}}
   n19 = "Transform (Modify) Bone - Bone: dyn_up_wing_r"(ComponentPose=@n18, Alpha=1, Translation="0,0,0", Rotation="5,30,0", Scale="1,1,1") {rotationMode=BMM_Additive, alphaScaleBiasClamp {bInterpResult=true}, boneToModify {boneName=dyn_up_wing_r}, rotation {pitch=5, yaw=30}}
   n20 = "Transform (Modify) Bone - Bone: dyn_down_wing_r"(ComponentPose=@n19, Alpha=1, Translation="0,0,0", Rotation="5,30,0", Scale="1,1,1") {rotationMode=BMM_Additive, alphaScaleBiasClamp {bInterpResult=true}, boneToModify {boneName=dyn_down_wing_r}, rotation {pitch=5, yaw=30}}
   n21 = "Transform (Modify) Bone - Bone: pack"(ComponentPose=@n20, Translation="0,0,1", Rotation="0,0,5") {
    rotationMode=BMM_Additive
    translationMode=BMM_Additive
    alphaScaleBiasClamp {bInterpResult=true}
    boneToModify {boneName=pack}
    rotation {roll=5}
    translation {z=1}
   }
   Save cached pose 'Adjustments'(Pose=Component To Local(ComponentPose=@n21)) {cacheName=Adjustments}
  ]}
 EventGraph {nodes [
   n1 = Event BlueprintUpdateAnimation()
   > Set Delta Time(Delta Time:float=@n1.DeltaTimeX) -> (Output_Get:float)
   > n2 = PlayerMovement(AnimInstance=Self-Reference(), XDotProductCoefficient=1, YDotProductCoefficient=1, ZDotProductCoefficient=1)
   > Set PlayerXMovement(PlayerXMovement:float=@n2) -> (Output_Get:float)
   > n3 = Sequence!(execute) -> (then_0, then_1, then_2:Exec, then_3:Exec, then_4:Exec)
   > n4 = Branch(Condition=Get bIsTacticalSprinting())
   n5 = FInterpTo(
    Current=Get wing_up_r()
    Target=30
    DeltaTime=float * float(A=Get Delta Time() -> (Delta Time:float), B=0.7) {is=title K2Node_PromotableOperator}
    InterpSpeed=7
   )
   > Set wing_up_r(wing_up_r:double=@n5) -> (Output_Get:double)
   > Set wing_down_r(wing_down_r:double=@n5) -> (Output_Get:double)
   n6 = Get Delta Time()
   > n7 = Set wing_up_l(wing_up_l:double=FInterpTo(Current=Get wing_up_l(), Target=-30, DeltaTime=float * float(A=@n6, B=0.7), InterpSpeed=7)) -> (Output_Get:double)
   > Set wing_down_l(wing_down_l:double=@n7) -> (Output_Get:double)
   n8 = Branch(execute=@n4.else, Condition=Get bIsSkydiveDiveMode())
   n9 = FInterpTo(Current=Get wing_up_r(), Target=33, DeltaTime=Get Delta Time(), InterpSpeed=8)
   > Set wing_up_r(wing_up_r=@n9)
   > Set wing_down_r(wing_down_r=@n9)
   > n10 = Set wing_up_l(wing_up_l=FInterpTo(Current=Get wing_up_l(), Target=-33, DeltaTime=Get Delta Time(), InterpSpeed=8))
   > Set wing_down_l(wing_down_l=@n10)
   n11 = Get bIsTacticalSprinting()
   n12 = Get bIsJumping()
   n13 = Get bIsFalling()
   n14 = Branch(execute=@n8.else, Condition=AND Boolean(A=OR Boolean(A=@n12, B=@n13), B=NOT Boolean(A=@n11)))
   n15 = FInterpTo(Current=Get wing_up_r(), Target=-25, DeltaTime=Get Delta Time(), InterpSpeed=12)
   > Set wing_up_r(wing_up_r=@n15)
   > Set wing_down_r(wing_down_r=@n15)
   > n16 = Set wing_up_l(wing_up_l=FInterpTo(Current=Get wing_up_l(), Target=25, DeltaTime=Get Delta Time(), InterpSpeed=12))
   > Set wing_down_l(wing_down_l=@n16)
   n17 = Branch(execute=@n14.else, Condition=AND Boolean(A=Get bIsCrouching(), B=Get bIsMoving2D()))
   n18 = FInterpTo(Current=Get wing_up_r(), Target=22, DeltaTime=Get Delta Time(), InterpSpeed=5)
   > Set wing_up_r(wing_up_r=@n18)
   > Set wing_down_r(wing_down_r=@n18)
   > n19 = Set wing_up_l(wing_up_l=FInterpTo(Current=Get wing_up_l(), Target=-22, DeltaTime=Get Delta Time(), InterpSpeed=5))
   > Set wing_down_l(wing_down_l=@n19)
   n20 = Set wing_up_r(execute=@n17.else, wing_up_r=FInterpTo(Current=Get wing_up_r(), Target=0, DeltaTime=Get Delta Time(), InterpSpeed=7.6))
   > Set wing_down_r(wing_down_r=@n20)
   > n21 = Set wing_up_l(wing_up_l=FInterpTo(Current=Get wing_up_l(), Target=0, DeltaTime=Get Delta Time(), InterpSpeed=8))
   > Set wing_down_l(wing_down_l=@n21)
   n22 = Set Alpha Output(
    execute=@n3.then_2
    Alpha Output:float=GetAlphaFromReferenceTransformDelta(
     SkeletalMeshComponent=GetOwnerSkeletalMeshComponent()
     BoneName=clavicle_l
     BaseName=upperarm_l
     DeltaComponentType=EulerY
     InputRange=(X=0,Y=6)
     TwistAxis="1,0,0"
    )
   ) -> (Output_Get:float)
   > Set Alpha arm rotations(Alpha arm rotations:bool=AND Boolean(A="To Bool (Double)"!(Value:double=@n22) -> (ReturnValue:bool), B=Get bIsSkydiveFloating())) -> (Output_Get:bool)
   n23 = Sequence!(execute=@n3.then_3) -> (then_0, then_1)
   n24 = Get bIsSkyDiving()
   n25 = Get PlayerXMovement() -> (PlayerXMovement:float)
   n26 = "To Boolean (Integer)"!(InInt:int=Round(A="Max (Float)"(A=@n25, B=0)) {is=title K2Node_CallFunction}) -> (ReturnValue:bool)
   n27 = NOT Boolean(A=Get bIsJumping())
   > n28 = Branch(Condition=AND Boolean(A=AND Boolean(A=@n26, B=@n27), B=NOT Boolean(A=@n24)))
   n29 = Get bIsSkyDiving()
   > n30 = Branch(Condition=AND Boolean(A=Get bIsCrouching(), B=NOT Boolean(A=@n29)))
   n31 = Set Right Move Wings(execute=@n28.else, Right Move Wings:double=0) -> (Output_Get:double)
   n32 = Get Delta Time()
   n33 = MapRangeClamped(Value=FInterpTo(Current=@n31, Target=float + float(A=@n31, B=1), DeltaTime=@n32, InterpSpeed=4), InRangeA=0, InRangeB=1)
   n34 = Set Right Move Wings(execute=@n30.else, Right Move Wings=@n33)
   Set Right Move Wings(execute=@n30.then, Right Move Wings=FInterpTo(Current=@n34, Target=float * float(A=@n33, B=1), DeltaTime=Get Delta Time(), InterpSpeed=25))
   n35 = Get bIsSkyDiving()
   n36 = "Min (Float)"(A:double=@n25, B:double=0) -> (ReturnValue:double) {is=title K2Node_CommutativeAssociativeBinaryOperator}
   n37 = "To Boolean (Integer)"(InInt=Round(A="Absolute (Float)"(A=@n36)))
   n38 = NOT Boolean(A=Get bIsJumping())
   n39 = Branch(execute=@n23.then_1, Condition=AND Boolean(A=AND Boolean(A=@n37, B=@n38), B=NOT Boolean(A=@n35)))
   > n40 = Branch(Condition=Get bIsCrouching())
   n41 = Set Left Move Wings(execute=@n39.else, Left Move Wings:double=0) -> (Output_Get:double)
   n42 = MapRangeClamped(Value=FInterpTo(Current=@n41, Target=float + float(A=@n41, B=1), DeltaTime=@n32, InterpSpeed=4), InRangeA=0, InRangeB=1)
   n43 = Set Left Move Wings(execute=@n40.else, Left Move Wings=@n42)
   Set Left Move Wings(execute=@n40.then, Left Move Wings=FInterpTo(Current=@n43, Target=float * float(A=@n42, B=1), DeltaTime=Get Delta Time(), InterpSpeed=25))
   n44 = Branch(execute=@n3.then_4, Condition=Get bIsInFrontEnd())
   > Set Sim Space Settings(Sim Space Settings:Sim Space Settings Structure=Get FrontEnd Sim Space Settings() -> (FrontEnd Sim Space Settings:Sim Space Settings Structure)) -> (Output_Get:Sim Space Settings Structure)
   Set Sim Space Settings(execute=@n44.else, Sim Space Settings=Get Game Sim Space Settings() -> (Game Sim Space Settings:Sim Space Settings Structure))
   n45 = GetOwnerSkeletalMeshComponent()
   n46 = GetSocketTransform!(self=@n45, InSocketName=upperarm_l, TransformSpace=RTS_Component) -> (ReturnValue_Location:Vector, ReturnValue_Rotation:Rotator, ReturnValue_Scale:Vector)
   n47 = "Clamp (Float)"(Value=float * float(A=BreakRotator(InRot=@n46.ReturnValue_Rotation).Pitch, B=-1), Min=-60, Max=1)
   Set Rotation Wing Up L(execute=@n3.then_1, Rotation Wing Up L:Rotator=MakeRotator(Yaw=@n47)) -> (Output_Get:Rotator)
   > Set Rotation Wing Down L(Rotation Wing Down L:Rotator=MakeRotator(Yaw=float / float(A=@n47, B=1.3))) -> (Output_Get:Rotator)
   n48 = GetSocketTransform(self=@n45, InSocketName=upperarm_r, TransformSpace=RTS_Component)
   n49 = "Clamp (Float)"(Value=float * float(A=BreakRotator(InRot=@n48.ReturnValue_Rotation).Pitch, B=-1), Min=-1, Max=25)
   > Set Rotation Wing Up R(Rotation Wing Up R:Rotator=MakeRotator(Yaw=@n49)) -> (Output_Get:Rotator)
   > Set Rotation Wing Down R(Rotation Wing Down R:Rotator=MakeRotator(Yaw=float / float(A=@n49, B=1.3))) -> (Output_Get:Rotator)
   TryGetPawnOwner()
   Get bIsSkyDiving()
  ]}
]
variables [
 Alpha arm rotations {type=boolean}
 Alpha Output {type=number}
 Delta Time {type=number}
 FrontEnd Sim Space Settings {
  type=SimSpaceSettings
  default {
   dampingAlpha=0
   maxAngularAcceleration=70
   maxAngularVelocity=20
   maxLinearAcceleration=3200
   maxLinearVelocity=850
   velocityScaleZ=0.6
   worldAlpha=1
   externalLinearDragV=<0 0 10>
  }
 }
 Game Sim Space Settings {
  type=SimSpaceSettings
  default {
   dampingAlpha=0
   maxAngularAcceleration=40
   maxAngularVelocity=10
   maxLinearAcceleration=3200
   maxLinearVelocity=800
   velocityScaleZ=0.6
   worldAlpha=1
   externalLinearDragV=<0 0 10>
  }
 }
 Left Move Wings {type=number}
 OnlySkyDi {type=boolean}
 PlayerXMovement {type=number}
 Return Value Rotation l {type=Rotator}
 Return Value Rotation_r {type=Rotator}
 Right Move Wings {type=number}
 Rotation Wing Down L {type=Rotator}
 Rotation Wing Down R {type=Rotator}
 Rotation Wing Up L {type=Rotator}
 Rotation Wing Up R {type=Rotator}
 Sim Space Settings {
  type=SimSpaceSettings
  default {
   dampingAlpha=0
   maxAngularAcceleration=40
   maxAngularVelocity=10
   maxLinearAcceleration=3200
   maxLinearVelocity=800
   velocityScaleZ=0.6
   worldAlpha=1
   externalLinearDragV=<0 0 10>
  }
 }
 wing_down_l {type=number}
 wing_down_r {type=number}
 wing_up_l {type=number}
 wing_up_r {type=number}
]
